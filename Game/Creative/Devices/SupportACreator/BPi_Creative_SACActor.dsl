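; BPi_Creative_SACActor is, by its name and path, a Blueprint interface for Support-A-Creator actors among the Creative devices.
; It declares a single function, GetSACDetails, which takes no inputs and returns three values: two default to empty and one to false.
; The file holds only this signature with default returns, so any real detail lookup lies outside it.
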(fn GetSACDetails
  (return "" "" false))
