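# Unlit additive material for a beam effect on the Support-A-Creator Creative device.
# SplineThicken sets the beam width through world position offset, combined with a horizontal push away from the camera.
# A half-sine across U times a linear V gradient, sharpened by a power, shapes the glow.
# Color from custom primitive data tints it, DepthFade softens intersections, and responsive AA is enabled.
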Material /Game/Creative/Devices/SupportACreator/Materials/M_SupportACreator_Beam
graph {
 properties {bEnableResponsiveAA=true, blendMode=BLEND_Additive, pixelDepthOffsetMode=PDOM_Legacy, shadingModel=MSM_Unlit, usedShadingModels=Unlit}
 nodes [
  n1 = ScalarParameter() {defaultValue=100, parameterName=SplineWidth}
  n2 = Normalize(VectorInput=Subtract(A=WorldPosition().XYZ, B=CameraPositionWS()))
  n3 = Constant3Vector() {constant=<rgba 1 1 0 0>}
  Add(
   A=MaterialFunctionCall(WidthBase=@n1, WidthTip=@n1) {materialFunction=&SplineThicken}.WorldPositionOffset
   B=Multiply(A=ScalarParameter() {defaultValue=20, parameterName=PushFromCameraXY}, B=Multiply(A=@n2, B=@n3))
  )
  n4 = MaterialFunctionCall(Float2=TextureCoordinate()) {materialFunction=&BreakOutFloat2Components}
  n5 = Multiply(A=@n4.G, B=Sine(None=@n4.R) {period=2})
  DepthFade(
   Opacity=Multiply(
    A=VectorParameter() {bUseCustomPrimitiveData=true, parameterName=Color, defaultValue=<rgba 1 0 0 0>}.RGB
    B=Power(Base=@n5) {constExponent=4}
   )
   FadeDistance=ScalarParameter() {defaultValue=30, parameterName=FadeDistance}
  )
 ]
}
