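# Lit material for the Support A Creator device button, with an animated procedural emissive display.
# Step-rotated UV1 becomes radial coordinates that drive a background glow and a segmented, patterned ring, combined with a sine-driven central pulse.
# The Pressed parameter blends this toward a flat bright value, which is then tinted and scaled by Fresnel.
# A packed mask drives specular, metallic and roughness, alongside a normal map and pattern-modulated base color.
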Material /Game/Creative/Devices/SupportACreator/Materials/M_SupportACreator_Button
graph {
 properties {pixelDepthOffsetMode=PDOM_Legacy, usedShadingModels=Default Lit}
 nodes [
  n1 = Time()
  n2 = ScalarParameter() {defaultValue=20, desc=20, parameterName=TimeSpeed}
  n3 = Rotator(Coordinate=TextureCoordinate() {coordinateIndex=1}, Time=Floor(None=Multiply(A=@n1, B=@n2) {constB=6.282}))
  n4 = MaterialFunctionCall(Float2=MaterialFunctionCall(Vector or UVs=ConstantBiasScale(None=@n3) {bias=-0.5, scale=2}) {materialFunction=&VectorToRadialValue}.Radial Coordinates) {materialFunction=&BreakOutFloat2Components}
  n5 = SphereMask(
   A=Cosine(None=@n4.R)
   B=ScalarParameter() {defaultValue=1, parameterName=BGGlowCenter}
   Radius=ScalarParameter() {defaultValue=0.2, parameterName=BGGlowRadius}
   Hardness=ScalarParameter() {parameterName=BGGlowHardness}
  )
  n6 = ScalarParameter() {defaultValue=0.04, parameterName=BGGlowScale}
  n7 = ScalarParameter() {defaultValue=20, parameterName=Steps}
  n8 = Multiply(A=@n4.R, B=@n7)
  n9 = Constant()
  n10 = Time() {period=1}
  n11 = Abs(None=Sine(None=Multiply(A=@n10) {constB=0.25}))
  n12 = SphereMask(A=Divide(A=Floor(None=@n8), B=@n7), B=@n9, Radius=Max(A=@n11) {constB=0.1}, Hardness=@n9)
  n13 = TextureCoordinate() {coordinateIndex=1}
  n14 = Constant() {r=0.5}
  n15 = Constant()
  n16 = Multiply(
   A=@n12
   B=SphereMask(
    A=SphereMask(A=@n13, B=@n14, Radius=@n14, Hardness=@n15)
    B=ScalarParameter() {defaultValue=0.45, desc=.75, parameterName=AnimatedRingCenter}
    Radius=ScalarParameter() {defaultValue=0.225, desc=.12, parameterName=AnimatedRingRadius}
    Hardness=ScalarParameter() {desc=.8, parameterName=AnimatedRingHardness}
   )
  )
  n17 = TextureCoordinate() {coordinateIndex=1}
  n18 = ScalarParameter() {defaultValue=40, parameterName=PatternTile}
  n19 = LinearInterpolate(
   A=ScalarParameter() {defaultValue=0.4, parameterName=PatternLow}
   Alpha=TextureSample(UVs=Multiply(A=@n17, B=@n18)) {texture=/Game/Athena/HUD/Materials/Victory/T_Square-Pattern2}.R
  )
  n20 = Power(Base=@n19, Exp=Add(A=@n12) {constB=0.05})
  n21 = Divide(A=@n7)
  n22 = Cosine(None=Frac(None=Multiply(A=@n4.R, B=@n21)))
  n23 = Constant()
  n24 = Constant() {r=1}
  n25 = ScalarParameter() {defaultValue=0.8, parameterName=LinesHardness}
  n26 = Multiply(A=Multiply(A=Multiply(A=@n12, B=@n20), B=SphereMask(A=@n22, B=@n23, Radius=@n24, Hardness=@n25)), B=@n19)
  n27 = TextureCoordinate() {coordinateIndex=1}
  n28 = Constant() {r=0.5}
  n29 = ScalarParameter() {defaultValue=0.28, parameterName=PulseRadius}
  n30 = Constant()
  n31 = Multiply(
   A=Power(Base=SphereMask(A=@n27, B=@n28, Radius=@n29, Hardness=@n30), Exp=ScalarParameter() {defaultValue=4, parameterName=PulseExponent})
   B=ScalarParameter() {defaultValue=1.5, parameterName=PulseScale}
  )
  n32 = ConstantBiasScale(None=@n11) {bias=0.35}
  n33 = Add(A=Multiply(A=Multiply(A=@n5, B=@n6), B=@n4.G), B=Add(A=Multiply(A=@n16, B=@n26), B=Multiply(A=@n31, B=@n32)))
  n34 = ScalarParameter() {parameterName=Pressed}
  n35 = ScalarParameter() {defaultValue=1.1, parameterName=FresnelBrightness}
  n36 = ScalarParameter() {defaultValue=4, parameterName=FresnelExponent}
  Multiply(
   A=Multiply(
    A=VectorParameter() {parameterName=EmissiveColor, defaultValue=<rgba 0.7165 3 0.5542 0>}.RGB
    B=LinearInterpolate(A=@n33, Alpha=@n34) {constB=2}
   )
   B=LinearInterpolate(
    A=ScalarParameter() {defaultValue=0.2, parameterName=FresnelLow}
    Alpha=MaterialFunctionCall(Brightness=@n35, Exponent=@n36) {materialFunction=&MF_Fresnel_FX}.Fresnel
   )
  )
  n37 = TextureSample() {desc=Masks, samplerType=SAMPLERTYPE_Masks, texture=&/T_SupportACreator_S}
  Multiply(A=@n37.R, B=ScalarParameter() {defaultValue=0.6, parameterName=Specular})
  Multiply(A=ScalarParameter() {defaultValue=0.4, parameterName=Metallic}, B=@n37.G)
  Multiply(A=@n37.B, B=ScalarParameter() {defaultValue=0.1875, parameterName=Roughness})
  Multiply(A=VectorParameter() {parameterName=BaseColor, defaultValue=<rgba 0.0051 0.0712 0.0212 1>}.RGB, B=@n19)
  TextureSample() {samplerType=SAMPLERTYPE_Normal, texture=&/T_SupportACreator_N}
 ]
}
